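; B_Prj_BattleLab_ElephantGarlic is a projectile Blueprint for the ElephantGarlic weapon among the Creative BattleLab items.
; When its ProjectileMovementComponent stops, it calls DoExplosion to detonate.
; Begin-overlap, tick and construction script are empty stubs, so this graph adds no other behavior.
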(event OnProjectileStop(ProjectileMovementComponent)
  (params (ImpactResult :type Hit%20Result%20Structure%20%28by%20ref%29))
  (Delivery|DoExplosion))

(event Collision|EventActorBeginOverlap
  (params (OtherActor :type Actor%20Object%20Reference)))

(event EventTick
  (params (DeltaSeconds :type Float%20%28single-precision%29)))

(fn ConstructionScript)
